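# Physics asset for the male medium StormAviator body's jacket, including a back panel, for simulating the hanging cloth's secondary motion.
# Seven bone chains around back, sides and front each run from a tiny master body, joined to spine_03 with free rotation, through two capsule segments.
# Segment joints allow much wider swing2 than swing1 or twist, so panels flex mainly on one axis.
# Spine and unconstrained thigh capsules supply body volumes around the jacket.
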PhysicsAsset /BRCosmetics/Characters/Player/Male/Medium/Bodies/M_MED_StormAviator/Meshes/M_MED_StormAviator_Physics_Jacket_WithBack
bodies {
 dyn_jacket_bk_01 [{shapeType=Capsule, length=10.2646, radius=2.7354, shapeName=dyn_jacket_bk_01_capsule, center=<5 0 0>, rotation=<rot 0 0 90>}]
 dyn_jacket_bk_02 [{shapeType=Capsule, length=9.0689, radius=3.0796, shapeName=dyn_jacket_bk_02_capsule, rotation=<rot 0 0 90>}]
 dyn_jacket_bk_MSTR [{shapeType=Capsule, length=0.505, radius=0.505, shapeName=dyn_jacket_bk_MSTR_capsule, rotation=<rot 0 -90 0>}]
 dyn_jacket_l_01 [{shapeType=Capsule, length=9.9708, radius=3.0292, shapeName=dyn_jacket_l_01_capsule, center=<5 0 0>, rotation=<rot 0 0 90>}]
 dyn_jacket_l_02 [{shapeType=Capsule, length=10.6046, radius=2.2522, shapeName=dyn_jacket_l_02_capsule, rotation=<rot 0 0 90>}]
 dyn_jacket_l_MSTR [{shapeType=Capsule, length=0.505, radius=0.505, shapeName=dyn_jacket_l_MSTR_capsule, rotation=<rot 0 -90 0>}]
 dyn_jacket_l_bk_01 [{shapeType=Capsule, length=9.9708, radius=3.0292, shapeName=dyn_jacket_l_bk_01_capsule, center=<5 0 0>, rotation=<rot 0 0 90>}]
 dyn_jacket_l_bk_02 [{shapeType=Capsule, length=9.0876, radius=2.6058, shapeName=dyn_jacket_l_bk_02_capsule, rotation=<rot 0 0 90>}]
 dyn_jacket_l_bk_MSTR [{shapeType=Capsule, length=0.505, radius=0.505, shapeName=dyn_jacket_l_bk_MSTR_capsule, rotation=<rot 0 -90 0>}]
 dyn_jacket_l_fr_01 [{shapeType=Capsule, length=9.9708, radius=3.0292, shapeName=dyn_jacket_l_fr_01_capsule, center=<5 0 0>, rotation=<rot 0 0 90>}]
 dyn_jacket_l_fr_02 [{shapeType=Capsule, length=9.387, radius=2.071, shapeName=dyn_jacket_l_fr_02_capsule, rotation=<rot 0 0 90>}]
 dyn_jacket_l_fr_MSTR [{shapeType=Capsule, length=0.505, radius=0.505, shapeName=dyn_jacket_l_fr_MSTR_capsule, rotation=<rot 0 -90 0>}]
 dyn_jacket_r_01 [{shapeType=Capsule, length=9.9708, radius=3.0292, shapeName=dyn_jacket_r_01_capsule, center=<5 0 0>, rotation=<rot 0 0 90>}]
 dyn_jacket_r_02 [{shapeType=Capsule, length=10.0562, radius=2.6659, shapeName=dyn_jacket_r_02_capsule, rotation=<rot 0 0 90>}]
 dyn_jacket_r_MSTR [{shapeType=Capsule, length=0.505, radius=0.505, shapeName=dyn_jacket_r_MSTR_capsule, rotation=<rot 0 -90 0>}]
 dyn_jacket_r_bk_01 [{shapeType=Capsule, length=9.9708, radius=3.0292, shapeName=dyn_jacket_r_bk_01_capsule, center=<5 0 0>, rotation=<rot 0 0 90>}]
 dyn_jacket_r_bk_02 [{shapeType=Capsule, length=9.5387, radius=2.6098, shapeName=dyn_jacket_r_bk_02_capsule, rotation=<rot 0 0 90>}]
 dyn_jacket_r_bk_MSTR [{shapeType=Capsule, length=0.505, radius=0.505, shapeName=dyn_jacket_r_bk_MSTR_capsule, rotation=<rot 0 -90 0>}]
 dyn_jacket_r_fr_01 [{shapeType=Capsule, length=9.9708, radius=3.0292, shapeName=dyn_jacket_r_fr_01_capsule, center=<5 0 0>, rotation=<rot 0 0 90>}]
 dyn_jacket_r_fr_02 [{shapeType=Capsule, length=9.2982, radius=2.0973, shapeName=dyn_jacket_r_fr_02_capsule, rotation=<rot 0 0 90>}]
 dyn_jacket_r_fr_MSTR [{shapeType=Capsule, length=0.505, radius=0.505, shapeName=dyn_jacket_r_fr_MSTR_capsule, rotation=<rot 0 -90 0>}]
 spine_01 [{shapeType=Capsule, length=7.0168, radius=11.2774, shapeName=spine_01_capsule}]
 spine_02 [{shapeType=Capsule, length=7.8125, radius=9.7993, shapeName=spine_02_capsule}]
 spine_03 [{shapeType=Capsule, length=5.7017, radius=10.4788, shapeName=spine_03_capsule, center=<0 -1.0613 0>}]
 thigh_l [{shapeType=Capsule, length=19.7575, radius=8.7772, shapeName=thigh_l_capsule, center=<-13 0 0>, rotation=<rot 90 0 0>}]
 thigh_r [{shapeType=Capsule, length=20.0102, radius=8.835, shapeName=thigh_r_capsule, center=<13 0 0>, rotation=<rot 90 0 0>}]
}
constraints [
 {
  bone1Name=spine_02
  bone2Name=spine_01
  swing1LimitDegrees=45
  swing1Motion=Free
  swing2LimitDegrees=45
  swing2Motion=Free
  twistLimitDegrees=45
  twistMotion=Free
 }
 {
  bone1Name=spine_03
  bone2Name=spine_02
  swing1LimitDegrees=45
  swing1Motion=Free
  swing2LimitDegrees=45
  swing2Motion=Free
  twistLimitDegrees=45
  twistMotion=Free
 }
 {
  bone1Name=dyn_jacket_bk_MSTR
  bone2Name=spine_03
  swing1LimitDegrees=45
  swing1Motion=Free
  swing2LimitDegrees=45
  swing2Motion=Free
  twistLimitDegrees=45
  twistMotion=Free
 }
 {bone1Name=dyn_jacket_bk_01, bone2Name=dyn_jacket_bk_MSTR, swing1LimitDegrees=3, swing2LimitDegrees=25, twistLimitDegrees=4}
 {bone1Name=dyn_jacket_bk_02, bone2Name=dyn_jacket_bk_01, swing1LimitDegrees=3, swing2LimitDegrees=25, twistLimitDegrees=4}
 {
  bone1Name=dyn_jacket_l_MSTR
  bone2Name=spine_03
  swing1LimitDegrees=45
  swing1Motion=Free
  swing2LimitDegrees=45
  swing2Motion=Free
  twistLimitDegrees=45
  twistMotion=Free
 }
 {bone1Name=dyn_jacket_l_01, bone2Name=dyn_jacket_l_MSTR, swing1LimitDegrees=3, swing2LimitDegrees=15, twistLimitDegrees=4}
 {bone1Name=dyn_jacket_l_02, bone2Name=dyn_jacket_l_01, swing1LimitDegrees=4, swing2LimitDegrees=15, twistLimitDegrees=4}
 {
  bone1Name=dyn_jacket_l_bk_MSTR
  bone2Name=spine_03
  swing1LimitDegrees=45
  swing1Motion=Free
  swing2LimitDegrees=45
  swing2Motion=Free
  twistLimitDegrees=45
  twistMotion=Free
 }
 {bone1Name=dyn_jacket_l_bk_01, bone2Name=dyn_jacket_l_bk_MSTR, swing1LimitDegrees=3, swing2LimitDegrees=25, twistLimitDegrees=4}
 {bone1Name=dyn_jacket_l_bk_02, bone2Name=dyn_jacket_l_bk_01, swing1LimitDegrees=3, swing2LimitDegrees=25, twistLimitDegrees=4}
 {
  bone1Name=dyn_jacket_l_fr_MSTR
  bone2Name=spine_03
  swing1LimitDegrees=45
  swing1Motion=Free
  swing2LimitDegrees=45
  swing2Motion=Free
  twistLimitDegrees=45
  twistMotion=Free
 }
 {bone1Name=dyn_jacket_l_fr_01, bone2Name=dyn_jacket_l_fr_MSTR, swing1LimitDegrees=5, swing2LimitDegrees=25, twistLimitDegrees=4}
 {bone1Name=dyn_jacket_l_fr_02, bone2Name=dyn_jacket_l_fr_01, swing1LimitDegrees=5, swing2LimitDegrees=25, twistLimitDegrees=4}
 {
  bone1Name=dyn_jacket_r_MSTR
  bone2Name=spine_03
  swing1LimitDegrees=45
  swing1Motion=Free
  swing2LimitDegrees=45
  swing2Motion=Free
  twistLimitDegrees=45
  twistMotion=Free
 }
 {bone1Name=dyn_jacket_r_01, bone2Name=dyn_jacket_r_MSTR, swing1LimitDegrees=3, swing2LimitDegrees=15, twistLimitDegrees=4}
 {bone1Name=dyn_jacket_r_02, bone2Name=dyn_jacket_r_01, swing1LimitDegrees=4, swing2LimitDegrees=15, twistLimitDegrees=4}
 {
  bone1Name=dyn_jacket_r_bk_MSTR
  bone2Name=spine_03
  swing1LimitDegrees=45
  swing1Motion=Free
  swing2LimitDegrees=45
  swing2Motion=Free
  twistLimitDegrees=45
  twistMotion=Free
 }
 {bone1Name=dyn_jacket_r_bk_01, bone2Name=dyn_jacket_r_bk_MSTR, swing1LimitDegrees=3, swing2LimitDegrees=25, twistLimitDegrees=4}
 {bone1Name=dyn_jacket_r_bk_02, bone2Name=dyn_jacket_r_bk_01, swing1LimitDegrees=3, swing2LimitDegrees=25, twistLimitDegrees=4}
 {
  bone1Name=dyn_jacket_r_fr_MSTR
  bone2Name=spine_03
  swing1LimitDegrees=45
  swing1Motion=Free
  swing2LimitDegrees=45
  swing2Motion=Free
  twistLimitDegrees=45
  twistMotion=Free
 }
 {bone1Name=dyn_jacket_r_fr_01, bone2Name=dyn_jacket_r_fr_MSTR, swing1LimitDegrees=5, swing2LimitDegrees=25, twistLimitDegrees=4}
 {bone1Name=dyn_jacket_r_fr_02, bone2Name=dyn_jacket_r_fr_01, swing1LimitDegrees=5, swing2LimitDegrees=25, twistLimitDegrees=4}
]
