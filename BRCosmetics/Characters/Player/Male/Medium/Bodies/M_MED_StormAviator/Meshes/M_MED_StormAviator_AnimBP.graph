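# Post-process-style AnimBlueprint for the StormAviator cosmetic body, adding secondary motion to the jacket, tags and ring.
# Copies the parent mesh pose through a deform Control Rig, offsetting rear jacket bones while skydiving.
# RigidBody sims branch on backbling (alternate physics asset plus spine constraints) and idle front-end state (zeroed motion scales), then feed AnimDynamics.
# EventGraph updates those flags, gravity override and fall-dependent motion scales.
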AnimBlueprint /BRCosmetics/Characters/Player/Male/Medium/Bodies/M_MED_StormAviator/Meshes/M_MED_StormAviator_AnimBP
target_skeleton=&/M_MED_StormAviator_Skeleton
graphs [
 AnimGraph {nodes [
   n1 = Anim Dynamics - Bone: dyn_tags!(ComponentPose=Local To Component(LocalPose=Use cached pose 'RB_Pose'()), Alpha=1, GravityOverride:Vector="0,0,-980", ExternalForce="0,0,0") -> (Pose) {
    angularBiasOverride=0.9999
    angularDampingOverride=0.9999
    angularSpringConstant=5000
    bOverrideAngularBias=true
    bOverrideAngularDamping=true
    bOverrideLinearDamping=true
    bUseGravityOverride=true
    bUsePlanarLimit=false
    gravityScale=-1
    linearDampingOverride=0.9999
    linearSpringConstant=5000
    simulationSpace=RootRelative
    alphaBoolBlend {blendOption=Linear}
    alphaScaleBiasClamp {scale=0.6}
    boundBone {boneName=dyn_tags}
    componentAppliedLinearAccClamp=<5000 5000 5000>
    componentLinearAccScale=<0.1 0.1 0.1>
    componentLinearVelScale=<0.1 0.1 0.1>
    gravityOverride {z=-980}
    physicsBodyDefinitions [{
      collisionType=CustomSphere
      sphereCollisionRadius=1
      boundBone {boneName=dyn_tags}
      boxExtents=<10 10 10>
      constraintSetup {angularLimitsMax=<0 22 2>, angularLimitsMin=<-30 -22 -2>, angularTarget=<0 0 0>}
      localJointOffset=<0 0 -5>
     }]
   }
   n2 = Anim Dynamics - Bone: dyn_ring_01(ComponentPose=@n1, Alpha=1, ExternalForce="0,0,0") {
    angularDampingOverride=0.99
    angularSpringConstant=500
    bOverrideAngularDamping=true
    bUseGravityOverride=true
    bUsePlanarLimit=false
    gravityScale=-1
    simulationSpace=RootRelative
    alphaBoolBlend {blendOption=Linear}
    boundBone {boneName=dyn_ring_01}
    componentAppliedLinearAccClamp=<5000 5000 5000>
    componentLinearAccScale=<0.2 0.2 0.2>
    componentLinearVelScale=<0.2 0.2 0.2>
    gravityOverride {z=-980}
    physicsBodyDefinitions [{
      collisionType=CustomSphere
      sphereCollisionRadius=1
      boundBone {boneName=dyn_ring_01}
      boxExtents=<10 5 5>
      constraintSetup {angularLimitsMax=<0 170 0>, angularLimitsMin=<0 0 0>, angularTarget=<0 0 0>}
      localJointOffset=<1 0 0>
     }]
   }
   Output Pose(Result=Component To Local(ComponentPose=@n2)) {layerGroup=None}
   n3 = Control Rig(Source=Copy Pose From Mesh() {bUseAttachedParent=true}) {
    alphaBoolBlend {blendOption=Linear}
    controlRigAssetReference {blueprintRigClass=&/M_MED_StormAviator_CtrlRig_Deform.M_MED_StormAviator_CtrlRig_Deform_C}
   }
   Save cached pose 'Start_Pose'(Pose=@n3) {cacheName=Start_Pose}
   n4 = "Transform (Modify) Bone - Bone: dyn_jacket_l_bk_MSTR"(ComponentPose=Local To Component(LocalPose=Use cached pose 'Start_Pose'()), Alpha=1, Translation="0,0,-2", Rotation="-15,0,0", Scale="1,1,1") {
    rotationMode=BMM_Additive
    rotationSpace=BCS_BoneSpace
    translationMode=BMM_Additive
    translationSpace=BCS_BoneSpace
    alphaBoolBlend {blendOption=Linear}
    boneToModify {boneName=dyn_jacket_l_bk_MSTR}
    rotation {pitch=-15}
    translation {z=-2}
   }
   n5 = "Transform (Modify) Bone - Bone: dyn_jacket_bk_MSTR"(ComponentPose=@n4, Alpha=1, Translation="0,0,-2", Rotation="-15,0,0", Scale="1,1,1") {
    rotationMode=BMM_Additive
    rotationSpace=BCS_BoneSpace
    translationMode=BMM_Additive
    translationSpace=BCS_BoneSpace
    alphaBoolBlend {blendOption=Linear}
    boneToModify {boneName=dyn_jacket_bk_MSTR}
    rotation {pitch=-15}
    translation {z=-2}
   }
   n6 = "Transform (Modify) Bone - Bone: dyn_jacket_r_bk_MSTR"(ComponentPose=@n5, Alpha=1, Translation="0,0,-2", Rotation="-15,0,0", Scale="1,1,1") {
    rotationMode=BMM_Additive
    rotationSpace=BCS_BoneSpace
    translationMode=BMM_Additive
    translationSpace=BCS_BoneSpace
    alphaBoolBlend {blendOption=Linear}
    boneToModify {boneName=dyn_jacket_r_bk_MSTR}
    rotation {pitch=-15}
    translation {z=-2}
   }
   Save cached pose 'Sky_Pose'(Pose=Component To Local(ComponentPose=@n6)) {cacheName=Sky_Pose}
   Save cached pose 'PreSim_Pose'(Pose=Blend Poses by bool(
     BlendPose_0=Use cached pose 'Sky_Pose'()
     BlendPose_1=Use cached pose 'Start_Pose'()
     BlendTime_0=0.7
     BlendTime_1=0.7
     bActiveValue=Get bIsSkyDiving()
    ) {blendType=Linear}) {cacheName=PreSim_Pose}
   n7 = RigidBody!(
    ComponentPose=Local To Component(LocalPose=Use cached pose 'PreSim_Pose'())
    Alpha
    OverrideWorldGravity:Vector=Get Gravity Override()
    ExternalForce
    ComponentLinearAccScale:Vector="0,0,0"
    ComponentLinearVelScale:Vector="0,0,0"
   ) -> (Pose) {
    bOverrideWorldGravity=true
    overridePhysicsAsset=&/M_MED_StormAviator_Physics_Jacket_WithBack
    simulationSpace=BaseBoneSpace
    alphaBoolBlend {blendOption=Linear}
    alphaScaleBiasClamp {scale=0.3}
    baseBoneRef {boneName=spine_03}
    componentAppliedLinearAccClamp=<4000 4000 4000>
    componentLinearAccScale=<0.1 0.1 0.1>
    componentLinearVelScale=<0.1 0.1 0.1>
   }
   n8 = Component To Local(ComponentPose=@n7)
   n9 = RigidBody(
    ComponentPose=Local To Component(LocalPose=Use cached pose 'PreSim_Pose'())
    OverrideWorldGravity=Get Gravity Override()
    ComponentLinearAccScale=Get ComponentLinearAccScale() -> (ComponentLinearAccScale:Vector)
    ComponentLinearVelScale=Get ComponentLinearVelScale() -> (ComponentLinearVelScale:Vector)
   ) {
    bOverrideWorldGravity=true
    overridePhysicsAsset=&M_MED_StormAviator_Physics_Jacket_WithBack
    simulationSpace=BaseBoneSpace
    alphaBoolBlend {blendOption=Linear}
    baseBoneRef {boneName=spine_03}
    componentAppliedLinearAccClamp=<4000 4000 4000>
    componentLinearAccScale=<0.35 0.35 0.35>
    componentLinearVelScale=<0.35 0.35 0.35>
   }
   n10 = Component To Local(ComponentPose=@n9)
   n11 = Get FE_Jacket_Physics() -> (FE_Jacket_Physics:bool)
   n12 = "Constraint - dyn_jacket_bk_01 (P)"(
    ComponentPose=Local To Component(LocalPose=Blend Poses by bool(BlendPose_0=@n8, BlendPose_1=@n10, bActiveValue=@n11) {blendType=Linear})
    Alpha=0.7
    ConstraintWeights_0=1
   ) {
    alphaBoolBlend {blendOption=Linear}
    boneToModify {boneName=dyn_jacket_bk_01}
    constraintSetup [{targetBone {boneName=spine_02}}]
    constraintWeights [1]
   }
   n13 = "Constraint - dyn_jacket_l_bk_01 (P)"(ComponentPose=@n12, Alpha=0.7, ConstraintWeights_0=1) {
    alphaBoolBlend {blendOption=Linear}
    boneToModify {boneName=dyn_jacket_l_bk_01}
    constraintSetup [{targetBone {boneName=spine_02}}]
    constraintWeights [1]
   }
   n14 = "Constraint - dyn_jacket_r_bk_01 (P)"(ComponentPose=@n13, Alpha=0.7, ConstraintWeights_0=1) {
    alphaBoolBlend {blendOption=Linear}
    boneToModify {boneName=dyn_jacket_r_bk_01}
    constraintSetup [{targetBone {boneName=spine_02}}]
    constraintWeights [1]
   }
   n15 = "Constraint - dyn_jacket_bk_02 (P)"(ComponentPose=@n14, Alpha=0.7, ConstraintWeights_0=1) {
    alphaBoolBlend {blendOption=Linear}
    boneToModify {boneName=dyn_jacket_bk_02}
    constraintSetup [{targetBone {boneName=spine_01}}]
    constraintWeights [1]
   }
   n16 = "Constraint - dyn_jacket_l_bk_02 (P)"(ComponentPose=@n15, Alpha=0.7, ConstraintWeights_0=1) {
    alphaBoolBlend {blendOption=Linear}
    boneToModify {boneName=dyn_jacket_l_bk_02}
    constraintSetup [{targetBone {boneName=spine_01}}]
    constraintWeights [1]
   }
   n17 = "Constraint - dyn_jacket_r_bk_02 (P)"(ComponentPose=@n16, Alpha=0.7, ConstraintWeights_0=1) {
    alphaBoolBlend {blendOption=Linear}
    boneToModify {boneName=dyn_jacket_r_bk_02}
    constraintSetup [{targetBone {boneName=spine_01}}]
    constraintWeights [1]
   }
   Save cached pose 'WithBack_Pose'(Pose=Component To Local(ComponentPose=@n17)) {cacheName=WithBack_Pose}
   n18 = RigidBody(
    ComponentPose=Local To Component(LocalPose=Use cached pose 'PreSim_Pose'())
    OverrideWorldGravity=Get Gravity Override()
    ComponentLinearAccScale=Get ComponentLinearAccScale()
    ComponentLinearVelScale=Get ComponentLinearVelScale()
   ) {
    bOverrideWorldGravity=true
    cachedBoundsScale=1.3
    overridePhysicsAsset=&/M_MED_StormAviator_Physics_Jacket
    simulationSpace=BaseBoneSpace
    alphaBoolBlend {blendOption=Linear}
    baseBoneRef {boneName=spine_03}
    componentAppliedLinearAccClamp=<4000 4000 4000>
    componentLinearAccScale=<0.35 0.35 0.35>
    componentLinearVelScale=<0.35 0.35 0.35>
   }
   n19 = Component To Local(ComponentPose=@n18)
   n20 = RigidBody(
    ComponentPose=Local To Component(LocalPose=Use cached pose 'PreSim_Pose'())
    OverrideWorldGravity=Get Gravity Override()
    ComponentLinearAccScale="0,0,0"
    ComponentLinearVelScale="0,0,0"
   ) {
    bOverrideWorldGravity=true
    overridePhysicsAsset=&M_MED_StormAviator_Physics_Jacket
    simulationSpace=BaseBoneSpace
    alphaBoolBlend {blendOption=Linear}
    alphaScaleBiasClamp {scale=0.3}
    baseBoneRef {boneName=spine_03}
    componentAppliedLinearAccClamp=<4000 4000 4000>
    componentLinearAccScale=<0.1 0.1 0.1>
    componentLinearVelScale=<0.1 0.1 0.1>
   }
   n21 = Component To Local(ComponentPose=@n20)
   n22 = Get FE_Jacket_Physics()
   Save cached pose 'RB_Pose'(Pose=Blend Poses by bool(
     BlendPose_0=Use cached pose 'WithBack_Pose'()
     BlendPose_1=Blend Poses by bool(BlendPose_0=@n21, BlendPose_1=@n19, bActiveValue=@n22) {blendType=Linear}
     BlendTime_0=0.7
     BlendTime_1=0.7
     bActiveValue=Get With_Back() -> (With_Back:bool)
    ) {blendType=Linear}) {cacheName=RB_Pose}
  ]}
 EventGraph {nodes [
   Event BlueprintUpdateAnimation()
   > n1 = Sequence!(execute) -> (then_0, then_1, then_2:Exec)
   > n2 = Gravity Override State Function DEPRECATED(CharacterPartAnimInstance=Self-Reference(), GameTimeInSeconds=GetGameTimeInSeconds(), GravityOverrideParams=Get Gravity Override Params())
   > Set Gravity Override(Gravity Override=@n2)
   n3 = SelectVector(A="0.25,0.25,0.25", B="0.5,0.5,0.5", bPickA=AND Boolean(A=Get bIsSkyDiving(), B=Get bIsFalling()))
   > Set ComponentLinearAccScale(ComponentLinearAccScale:Vector=@n3) -> (Output_Get:Vector)
   > Set ComponentLinearVelScale(ComponentLinearVelScale:Vector=@n3) -> (Output_Get:Vector)
   n4 = CheckForBackbling()
   Set With_Back(execute=@n1.then_1, With_Back:bool=OR Boolean!(A=@n4, B=@n4.bIsBackblingACape, C:bool=@n4.bIsBackblingLong) -> (ReturnValue)) -> (Output_Get:bool)
   n5 = NOT Boolean(A=Get bIsPlayingEmote())
   n6 = NOT Boolean(A=Get bIsSkyDiving())
   n7 = Branch(execute=@n1.then_2, Condition=AND Boolean(A=Get bIsInFrontEnd(), B=AND Boolean(A=@n6, B=@n5)))
   > Set FE_Jacket_Physics(FE_Jacket_Physics:bool=true) -> (Output_Get:bool)
   Set FE_Jacket_Physics(execute=@n7.else, FE_Jacket_Physics=false)
   Event BlueprintInitializeAnimation()
   > Set SkelMesh(SkelMesh=GetOwnerSkeletalMeshComponent())
  ]}
]
variables [
 ComponentLinearAccScale {type=Vector}
 ComponentLinearVelScale {type=Vector}
 FE_Jacket_Physics {type=boolean}
 Gravity Override {type=Vector}
 Gravity Override Params {type=GravityOverrideParamsStruct}
 SkelMesh {type=SkeletalMeshComponent}
 With_Back {type=boolean}
]
